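# Test is a minimal two-sided Default Lit material in the WildEstate tree foliage TestShowCase folder.
# It is flagged for use with Nanite, and automatic usage setting in the editor is turned off.
# Its graph holds a single scalar constant with no connection shown to any material output, so it defines no textures, parameters or shading logic.
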Material /WildEstate/Foliage/Trees/WildEstate_C/Materials/TestShowCase/Test
graph {
 properties {bAutomaticallySetUsageInEditor=false, bUsedWithNanite=true, twoSided=true, usedShadingModels=Default Lit}
 nodes [Constant() {r=0.5}]
}
